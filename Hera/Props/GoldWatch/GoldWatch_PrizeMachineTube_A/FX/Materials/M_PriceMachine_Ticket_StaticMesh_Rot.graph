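# Masked, two-sided lit material for prize machine ticket static meshes, built to make the tickets spin and swirl.
# Rotation about the object pivot runs on time, with vertex red blending two speeds.
# A curl offset, added to that rotation, comes from time-scrolled, actor-relative vector noise, with vertex alpha blending its scale and intensity.
# The graph shows no links to material outputs, including from the ticket texture, and holds leftover rotation nodes.
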Material /Hera/Props/GoldWatch/GoldWatch_PrizeMachineTube_A/FX/Materials/M_PriceMachine_Ticket_StaticMesh_Rot
graph {
 properties {bAutomaticallySetUsageInEditor=false, blendMode=BLEND_Masked, twoSided=true, usedShadingModels=Default Lit}
 nodes [
  n1 = Multiply(
   A=Time()
   B=LinearInterpolate(
    A=ScalarParameter() {defaultValue=-0.2, parameterName=RotationSpeed}
    B=ScalarParameter() {defaultValue=-0.2, parameterName=RotationSpeedLow}
    Alpha=VertexColor().R
   )
  )
  n2 = WorldPosition() {worldPositionShaderOffset=WPT_ExcludeAllShaderOffsets}
  Add(
   A=NamedRerouteUsage()
   B=RotateAboutAxis(
    NormalizedRotationAxis=VectorParameter() {parameterName=RotationAxis, defaultValue=<rgba 0 0 1 0>}.RGB
    RotationAngle=Frac(None=@n1)
    PivotPoint=MaterialFunctionCall() {materialFunction=&ObjectPivotPoint}.Object Pivot Location
    Position=@n2.XYZ
   )
  )
  ComponentMask(None=ObjectPositionWS()) {b=true}
  n3 = ScalarParameter() {defaultValue=50, parameterName=CurlIntensity}
  n4 = ScalarParameter() {defaultValue=100, parameterName=CurlIntensityHigh}
  n5 = VertexColor()
  n6 = WorldPosition()
  n7 = ActorPositionWS()
  n8 = Time()
  n9 = ScalarParameter() {defaultValue=1, parameterName=CurlSpeed}
  n10 = Divide(
   A=Subtract(A=Subtract(A=@n6.XYZ, B=@n7), B=Multiply(A=@n8, B=@n9))
   B=LinearInterpolate(
    A=ScalarParameter() {defaultValue=1000, parameterName=CurlScale}
    B=ScalarParameter() {defaultValue=500, parameterName=CurlScale_Low}
    Alpha=@n5.A
   )
  )
  NamedRerouteDeclaration(None=Multiply(A=LinearInterpolate(A=@n3, B=@n4, Alpha=@n5.A), B=VectorNoise(World Position=@n10) {noiseFunction=VNF_VectorALU})) {name=Curl, nodeColor=<rgba 1 0.6562 0 1>}
  RotateAboutAxis(
   NormalizedRotationAxis=Transform(None=VectorParameter() {parameterName=RotationAxis, defaultValue=<rgba 0 0 1 0>}.RGB) {transformSourceType=TRANSFORMSOURCE_Local}
   RotationAngle=Multiply(A=Time(), B=ScalarParameter() {defaultValue=-0.2, parameterName=RotationSpeed})
   PivotPoint=ActorPositionWS()
   Position=WorldPosition().XYZ
  )
  ScalarParameter() {defaultValue=-0.1, parameterName=RotationSpeed_Low}
  TextureSample() {texture=/Hera/Props/GoldWatch/GoldWatch_PrizeMachineTube_A/FX/Texture/T_TicketKiosk_Ticket_D_01}
  Transform(None=@n2.XYZ) {transformSourceType=TRANSFORMSOURCE_World, transformType=TRANSFORM_Local}
 ]
}
