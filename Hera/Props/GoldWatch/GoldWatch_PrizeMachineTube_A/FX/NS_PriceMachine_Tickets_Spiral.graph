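# Niagara effect for the GoldWatch prize machine tube prop that, going by its name, shows tickets swirling in a spiral.
# One GPU emitter continuously spawns short-lived mesh particles in a cylinder volume, scaled up and back down over their life by a curve.
# Randomized RotateAroundPoint orbits, random mesh spin and panning curl noise with a velocity clamp create the swirl.
# Point attraction and accurate velocity modules are present but disabled.
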NiagaraSystem /Hera/Props/GoldWatch/GoldWatch_PrizeMachineTube_A/FX/NS_PriceMachine_Tickets_Spiral
dependencies {usedDataInterfaces [&NiagaraDataInterfaceCurve], usedDynamicInputs [], usedModules [&SystemState]}
emitters {Minimal {
  simTarget=GPUComputeSim
  emitterUpdateScript [
   &EmitterState()
   &SpawnRate(SpawnRate=100, Use Spawn Probability=false)
  ]
  particleSpawnScript [
   &InitializeParticle(
    Color Mode=::NewEnumerator0
    Lifetime=2
    Mesh Scale Mode=::NewEnumerator3
    Mesh Uniform Scale=3
    UsePositionOffset=false
    Write Material Random=true
   )
   &ShapeLocation(Cylinder Radius=150, MaskSpawnGroup=false, Override Seed=false, Shape Primitive=::NewEnumerator1, Surface Only=false, UseOwnerScale=true)
  ]
  particleUpdateScript [
   &ParticleState()
   &ScaleMeshSize(Scale Factor=&Multiply_VectorByFloat(
     Vector=<1 1 1>
     Float=&FloatFromCurve(FloatCurve=DataInterface() {propertyValues {Curve {
         defaultValue=1
         keys [
          {interpMode=RCIM_Cubic, tangentMode=RCTM_User, time=0, value=0}
          {arriveTangent=2, leaveTangent=2, time=0.2154, value=1}
          <at 0.5 1>
          {arriveTangent=-2, leaveTangent=-2, time=0.7676, value=1}
          <at 1 0 RCIM_Cubic>
         ]
        }}})
    ))
   &UpdateMeshOrientation(Rotation Vector=&RandomRangeVector(Minimum=<0 0 0>, Maximum=<1 1 1>))
   &RotateAroundPoint(
    Derive the Rotation Axis=false
    Radius:NiagaraFloat=&RandomRangeFloat(Minimum=25, Maximum=50)
    Rotation Center:NiagaraPosition=&RandomRangeVector(Minimum=<0 0 -100>, Maximum=<0 0 100>)
    Rotation Phase:NiagaraFloat=&RandomRangeFloat(Minimum=0, Maximum=1)
    Rotation Rate:NiagaraFloat=&RandomRangeFloat(Minimum=0.35, Maximum=0.5)
   )
   &CurlNoiseForce(Bias Noise=false, Noise Frequency=25, Noise Strength=500, Pan Noise=true, Pan Noise Field=<0 0 1>)
   &PointAttractionForce(
    Attraction Radius=1500
    Attraction Strength=-100
    Falloff Exponent=1
    Kill Within Radius=false
    Module.Use Falloff=true
    UseAttractorOffset=false
   ) {enabled=false}
   &SolveForcesAndVelocity(Clamp Velocity=true, Limit Acceleration=false, Speed Limit=250)
   &CalculateAccurateVelocity() {enabled=false}
  ]
  rendererClasses [&NiagaraMeshRendererProperties]
  renderers [{rendererClass=&NiagaraMeshRendererProperties, rendererIndex=0}]
 }}
summary {
 systemName=NS_PriceMachine_Tickets_Spiral
 emitters [{emitterName=Minimal, simTarget=GPUComputeSim, rendererClasses [&NiagaraMeshRendererProperties]}]
 userVariables []
}
